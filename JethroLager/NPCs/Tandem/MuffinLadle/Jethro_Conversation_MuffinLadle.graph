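# Conversation database for the MuffinLadle NPC in Jethro, serving as the root dialogue when a player talks to it.
# One entry point leads to a greeting Speech node whose contextual lines react to first-ever meeting, per-match conversation count and specific player outfits, with priority curve rows on first-meeting and outfit reactions.
# The greeting fans out through link nodes to separate entries for quests, Jethro services, Season 35 found, deluxe services and MuffinLadle's own dialogue.
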ConversationDatabase /JethroLager/NPCs/Tandem/MuffinLadle/Jethro_Conversation_MuffinLadle
entry_points [{tag=Jethro.Conversation.Entry.MuffinLadle, destinations [#1]}]
nodes [
 ConversationEntryPointNode_7 {
  id=#1
  class=ConversationEntryPointNode
  outputs [#2]
  properties {entryTag {tagName=Jethro.Conversation.Entry.MuffinLadle}}
 }
 Speech_C_4 {
  id=#2
  class=Speech_C
  outputs [
   #3
   #4
   #5
   #6
   #7
  ]
  properties {
   nodeName=""
   generalConfig {
    defaultMessage=You again. What is it?
    contextualMessages [
     {
      message="What?!"
      contextRequirements [{requirement {class=FortControllerRequirement_IsFirstEverConversationWithNPC}}]
      priority {curve {rowName=Tandem.Speech.Priority.FirstEverInteraction}}
     }
     {
      message=You again. What is it?
      contextRequirements [{requirement {
         class=FortControllerRequirement_CheckNPCMatchConversationCount
         properties {nPCUniqueIdTag {tagName=AISpawnerData.Type.Tandem.Jethro.MuffinLadle}}
        }}]
      priority {curve {rowName=None}}
     }
     {
      message="Have you brought me a lightsaber, or are you here to waste my time?"
      contextRequirements [{requirement {
         class=FortControllerRequirement_CheckNPCMatchConversationCount
         properties {nPCUniqueIdTag {tagName=AISpawnerData.Type.Tandem.Jethro.MuffinLadle}}
        }}]
      priority {curve {rowName=None}}
     }
     {
      message="...\r\n"
      contextRequirements [{requirement {
         class=FortControllerRequirement_HasCID
         properties {allowedCIds [
           &CID_A_422_Athena_Commando_M_Realm
           /BRCosmetics/Athena/Items/Cosmetics/Characters/Character_SequinPie
          ]}
        }}]
      priority {curve {rowName=Tandem.Speech.Priority.OutfitReaction}}
     }
     {
      message=...
      contextRequirements [{requirement {class=FortControllerRequirement_HasCID, properties {allowedCIds [/BRCosmetics/Athena/Items/Cosmetics/Characters/Character_ClawPad_Host]}}}]
      priority {curve {rowName=Tandem.Speech.Priority.OutfitReaction}}
     }
     {
      message=Get out of my sight.
      contextRequirements [{requirement {
         class=FortControllerRequirement_HasCID
         properties {allowedCIds [/BRCosmetics/Athena/Items/Cosmetics/Characters/CID_752_Athena_Commando_M_Comet]}
        }}]
      priority {curve {rowName=Tandem.Speech.Priority.OutfitReaction}}
     }
    ]
   }
  }
 }
 ConversationLinkNode_4 {
  id=#7
  class=ConversationLinkNode
  properties {remoteEntryTag {tagName=Jethro.Conversation.Entry.MuffinLadle.Dialogue}}
 }
 ConversationLinkNode_3 {id=#6, class=ConversationLinkNode, properties {remoteEntryTag {tagName=Conversation.Entry.DeluxeServices}}}
 ConversationLinkNode_2 {
  id=#5
  class=ConversationLinkNode
  properties {remoteEntryTag {tagName=Conversation.Entry.Season35.Found.MuffinLadle}}
 }
 ConversationLinkNode_0 {id=#4, class=ConversationLinkNode, properties {remoteEntryTag {tagName=Conversation.Entry.JethroServices}}}
 ConversationLinkNode_1 {id=#3, class=ConversationLinkNode, properties {remoteEntryTag {tagName=Conversation.Entry.NPCLibrary.Quests}}}
]
